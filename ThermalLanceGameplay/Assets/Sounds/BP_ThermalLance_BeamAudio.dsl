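; Audio actor for the Thermal Lance beam, playing fire, loop and end sounds plus repeating sound indicators for non-local players.
; UpdateBeam stretches an audio shape line along the beam, keeps the loop at its start and feeds fuel to the edge audio's Overheat parameter.
; The actor follows the endpoint, where a lazily created surface component switches sound by hit actor: water, pawn or other.
; Tag-gated vault events briefly flag weakpoint hits to override that surface.
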(event Custom|SetUpVaultEvents
  ; Set up vault events
  (Ability|Tasks|WaitGameplayTagAddToActor (Variables|Actor|GetInstigator) (Variables|Default|GetVaultParticipantTag) true
    (:Added
      (bind (_asynctaskproxy _eventrouter) (Class|AsyncActionWaitforPlayspaceEventRouter|WaitforPlayspaceEventRouter 0 "/VividRazorCoreGameplay/Gameplay/Interfaces/VividRazorBasePlayspace.VividRazorBasePlayspace_C" true true)
        (:OnEventRouterAvailable
          (Utilities|FlowControl|Sequence
            (:then_0
              (Ability|Tasks|WaitGameplayTagRemoveFromActor (Variables|Actor|GetInstigator) (Variables|Default|GetVaultParticipantTag) true
                (:Removed
                  (GameplayEventRouter|StopListeningtoEvent (Variables|Default|GetWeakpointDamagedTask))
                  (CallFunction|SetUpVaultEvents))))
            (:then_1
              (bind (_asynctaskproxy_2 _payload) (GameplayEventRouter|StartListeningToEvent _eventrouter "/VividRazorCoreGameplay/Gameplay/Structs/Struct_VividRazorCore_Vault_Weapoint_Damaged_Payload.Struct_VividRazorCore_Vault_Weapoint_Damaged_Payload")
                (:then
                  (Variables|Default|SetWeakpointDamagedTask _asynctaskproxy_2))
                (:OnEventReceived
                  (if (== (.DamageCauser_30_2426A9E04DD5C75D5989B89B6CE43604 (Utilities|Struct|BreakStructVividRazorCoreVaultWeapointDamagedPayload _payload)) (Variables|Transient|GetOwnerWeaponActor))
                    (Variables|Default|SetShootingWeakPoint true)
                    (Utilities|FlowControl|RetriggerableDelay 0.1)
                    (Variables|Default|SetShootingWeakPoint false)))))))))))

(event EventBeginPlay
  (bind _instigator (Variables|Actor|GetInstigator))
  (bind _returnvalue (Audio|SpawnSoundAttached (Variables|Default|GetWeaponLoop_Sound) (Variables|Default|GetDefaultSceneRoot)))
  (Variables|Default|SetWeaponAudioComponent _returnvalue)
  (Default|SpawnSoundIndicator)
  (bind _returnvalue_2 (Utilities|Time|SetTimerbyEvent (EventDispatchers|CreateEvent :!event SpawnSoundIndicator) (Math|Float|SafeDivide 1.0 (FortWeapon|GetFiringRate (Variables|Transient|GetOwnerWeaponActor))) true))
  (Variables|Transient|SetSpawnSoundIndicatorTimerHandle _returnvalue_2)
  (CallFunction|SetUpVaultEvents)
  (Utilities|IsValid _instigator
    (:Is%20Valid
      (Audio|SpawnSoundAttached (select (Pawn|IsLocallyViewed _instigator) (Variables|Default|GetWeapon_Fire_1P) (Variables|Default|GetWeapon_Fire_3P)) (Variables|/Script/Engine.Actor|GetRootComponent _instigator)))))

(event Custom|UpdateBeam
  (params (StartPoint :type Vector) (EndPoint :type Vector) (FuelRemaining :type Float%20%28double-precision%29) (HitActor :type Actor%20Object%20Reference))
  (bind _audioshapeline (Variables|Default|GetAudioShapeLine))
  (bind _returnvalue (Control|GetEdgeAudioComponent _audioshapeline))
  (CallFunction|UpdateBeamLocation StartPoint EndPoint)
  (Utilities|IsValid _returnvalue
    (:Is%20Valid
      (Audio|Parameter|SetFloatParameter _returnvalue "Overheat" FuelRemaining :!class /Script/Engine.AudioComponent)
      (bind (_change _surface) (CallFunction|ActortoSurface HitActor))
      (if _change
        (bind _output_get (Variables|Transient|SetPreviousMat _surface))
        (bind _surfacesoundcomponent (CallFunction|AddSurfaceSoundComponent))
        (Class|SurfaceSoundComponent|SurfaceTypeChanged _surfacesoundcomponent _output_get)))))

(event EventDestroyed
  (Audio|SpawnSoundAttached (Variables|Default|GetBeamEnd_Sound) (Variables|/Script/Engine.Actor|GetRootComponent (Variables|Actor|GetInstigator)))
  (Utilities|Time|ClearandInvalidateTimerbyHandle (Variables|Transient|GetSpawnSoundIndicatorTimerHandle))
  (bind _weaponaudiocomponent (Variables|Default|GetWeaponAudioComponent)
    (:then
      (Audio|Components|Audio|FadeOut _weaponaudiocomponent 0.25 0.0 "Linear" :!class /Script/Engine.AudioComponent))
    (:else)))

(fn ActorToSurface
  (params (HitActor :type Actor%20Object%20Reference))
  (bind _hitactor (Variables|GetHitActor))
  (if (and (== HitActor (Variables|Transient|GetPreviousActor)) (and (Variables|Default|GetShootingWeakPoint) (Utilities|Enum|Equal(Enum) (Variables|Transient|GetPreviousMat) "SurfaceType3")))
    (return false (Variables|Transient|GetPreviousMat))
    (else
      (Variables|GetHitActor
        (:then
          (Utilities|FlowControl|Sequence
            (:then_0
              (Variables|Transient|SetPreviousActor (Variables|GetHitActor))
              (if (Variables|Default|GetShootingWeakPoint)
                (LocalVariables|SetNewSurface "SurfaceType3")
                (else
                  (Utilities|IsA(soft) _hitactor "/Script/Water.WaterBody"
                    (:True
                      (LocalVariables|SetNewSurface "SurfaceType8"))
                    (:False
                      (Utilities|IsA(soft) _hitactor "/Script/FortniteGame.FortPawn"
                        (:True
                          (LocalVariables|SetNewSurface "SurfaceType4"))
                        (:False
                          (LocalVariables|SetNewSurface "SurfaceType10"))))))))
            (:then_1
              (return (Utilities|Enum|NotEqual(Enum) (Variables|Transient|GetPreviousMat) (LocalVariables|GetNewSurface)) (LocalVariables|GetNewSurface)))))
        (:else
          (Variables|Transient|SetPreviousActor 0)
          (return true "SurfaceType_Default"))))))

(fn AddSurfaceSoundComponent
  (bind _surface_sound_component (Variables|Default|GetSurface%20Sound%20Component)
    (:then
      (return _surface_sound_component))
    (:else
      (bind _returnvalue (Game|AddComponentbyClass 0 "/Game/Sounds/Fort_Audio_Blueprints/SurfaceTypeSound/BP_SurfaceSoundComponent.BP_SurfaceSoundComponent_C" (Variables|Default|GetDefaultSceneRoot) (Variables|Default|GetEndpoint_Sound) (Variables|Default|GetSurface%20Sound%20Data%20Table)))
      (bind _output_get (Variables|Default|SetSurface%20Sound%20Component _returnvalue))
      (return _output_get))))

(fn SpawnSoundIndicator
  ;[0.02051 0.132 0.15 0.5] Spawn the sound indicator
  (bind _weaponloop_sound (Variables|Default|GetWeaponLoop_Sound)
    (:then
      (bind _instigator (Variables|Actor|GetInstigator)
        (:then
          (if (and (not (Pawn|IsLocallyControlled _instigator)) (not (Pawn|IsLocallyViewed _instigator)))
            (Fortnite|ShowSoundIndicatorLocalforTeam :TrackedActor _instigator :Location (Transformation|GetActorLocation _instigator) :MaxAudibleDistance (Variables|/Script/Engine.SoundBase|GetMaxDistance _weaponloop_sound) :IndicatorType "Gunshot" :Team (Team|GetActorTeam (Variables|Actor|GetInstigator)) :Affiliations (Utilities|Array|MakeArray "Neutral" "Hostile" "Friendly") :Tint "(R=0.000000,G=0.000000,B=0.000000,A=0.000000)" :AttenuationMaxRangeMultiplier 1.0 :AttenuationIntensityMultiplier 1.0)))
        (:else)))
    (:else)))

(fn UpdateBeamLocation
  (params (StartPoint :type Vector) (Endpoint :type Vector))
  (if (not (Components|Activation|IsActive (Variables|Default|GetAudioShapeLine)))
    (Components|Activation|Activate (Variables|Default|GetAudioShapeLine) false))
  (Control|SetEndPoint (Variables|Default|GetAudioShapeLine) (Variables|GetEndpoint))
  (Control|SetStartPoint (Variables|Default|GetAudioShapeLine) (Variables|GetStartPoint))
  (Transformation|SetActorLocation (Variables|GetEndpoint) true)
  (bind _weaponaudiocomponent (Variables|Default|GetWeaponAudioComponent)
    (:then
      (Transformation|SetWorldLocation _weaponaudiocomponent (Variables|GetStartPoint) true false))
    (:else)))

(fn ConstructionScript)
