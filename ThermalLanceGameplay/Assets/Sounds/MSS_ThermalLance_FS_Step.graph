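# One-shot stereo MetaSound for Thermal Lance footstep foley, fed by character and foley-action interface inputs.
# On play, a Stance check either triggers a shared array one-shot player with the Run waves, or skips straight to OnFinished.
# A parallel character-affiliation check routes full or quarter gain into a stereo mixer on the player's output.
# OnFinished also fires when the played wave ends.
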MetaSoundSource /ThermalLanceGameplay/Assets/Sounds/MSS_ThermalLance_FS_Step
inputs [
 Character.Affiliation {type=Enum:FortAudioAffiliation}
 Character.Health {type=Float}
 Character.UniqueID {type=Int32}
 FoleyAction.WaveAssets {type=WaveAsset:Array}
 Stance {type=Int32}
 UE.Source.OnPlay {type=Trigger}
]
interfaces [Fort.Character:1.0, Fort.FoleyAction:1.0, UE.OutputFormat.Stereo:1.0, UE.Source.OneShot:1.0, UE.Source:1.1]
outputs [
 "UE.OutputFormat.Stereo.Audio:0" {type=Audio}
 "UE.OutputFormat.Stereo.Audio:1" {type=Audio}
 UE.Source.OneShot.OnFinished {type=Trigger}
]
pages [Default {
  input_defaults {
   Character.Affiliation=0
   Character.Health=0
   Character.UniqueID=0
   Stance=0
   UE.Source.OnPlay=null
   FoleyAction.WaveAssets [
    /ThermalLanceGameplay/Assets/Sounds/Waves/ThermalLance_Run_01
    /ThermalLanceGameplay/Assets/Sounds/Waves/ThermalLance_Run_02
    /ThermalLanceGameplay/Assets/Sounds/Waves/ThermalLance_Run_03
    /ThermalLanceGameplay/Assets/Sounds/Waves/ThermalLance_Run_04
    /ThermalLanceGameplay/Assets/Sounds/Waves/ThermalLance_Run_05
    /ThermalLanceGameplay/Assets/Sounds/Waves/ThermalLance_Run_06
    /ThermalLanceGameplay/Assets/Sounds/Waves/ThermalLance_Run_07
    /ThermalLanceGameplay/Assets/Sounds/Waves/ThermalLance_Run_08
    /ThermalLanceGameplay/Assets/Sounds/Waves/ThermalLance_Run_09
   ]
  }
  nodes [
   n1 = TriggerCompare.Clamp:Int32(B=^Stance, Compare=^UE.Source.OnPlay) {also {defaults {A=2, Type=1}}}
   n2 = /MetaSoundWizard/MetaSounds/One-shots/Array/MSS_Oneshot_Array_Stereo(Sounds=^FoleyAction.WaveAssets, UE.Source.OnPlay=@n1.True)
   n3 = &TriggerCompareAffiliation(A=^Character.Affiliation, Compare=^UE.Source.OnPlay) {also {defaults {B=2}}}
   TriggerRouteTriggerRouteFloat2_0 = "TriggerRoute.Trigger Route (Float, 2):Float"(Set 1=@n3.False, Set 0=@n3.True) {also {defaults {Value 0=1, Value 1=0.25}}}
   StereoMixer_One_0 = /FortMetasound/Patches/StereoMixer_One(InL=@n2."UE.OutputFormat.Stereo.Audio:0", InR=@n2."UE.OutputFormat.Stereo.Audio:1", "Volume(Lin)"=@TriggerRouteTriggerRouteFloat2_0.Value)
   TriggerAnyTriggerAny2_0 = "TriggerAny.Trigger Any (2)"(In 0=@n2.UE.Source.OneShot.OnFinished, In 1=@n1.False)
  ]
  links [
   {dst="@UE.OutputFormat.Stereo.Audio:0", src=StereoMixer_One_0.OutL}
   {dst="@UE.OutputFormat.Stereo.Audio:1", src=StereoMixer_One_0.OutR}
   {dst="@UE.Source.OneShot.OnFinished", src=TriggerAnyTriggerAny2_0.Out}
  ]
 }]
properties {
 assetImportData=&self:AssetImportData
 attenuationSettings=/ThermalLanceGameplay/Assets/Sounds/Settings/SA_ThermalLance_Foley
 maxDistance=2250
 numChannels=2
 outputFormat=Stereo
 sampleRate=48000
 soundClassObject=&Fort_WeaponFoley
 volume=0.7
}
