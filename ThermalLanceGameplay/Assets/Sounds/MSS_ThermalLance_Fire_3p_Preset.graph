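# MetaSound Source preset that plays third-person fire audio for the Thermal Lance weapon, inheriting its graph from MSS_Base_Weapon_Fire_3p_Simple.
# It supplies four-variant close, medium and distant wave sets, with per-layer gains and a shared ranged-weapon distance-curve bank.
# Inputs also set random pitch, volume and EQ ranges, burst and retrigger timing, and friendly-fire volume.
# Output is stereo, using Fortnite weapon attenuation, third-person sound class and gunfire concurrency.
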MetaSoundSource /ThermalLanceGameplay/Assets/Sounds/MSS_ThermalLance_Fire_3p_Preset
preset_parent=/FortMetaSoundWeapon/Items/Metasounds/MSS_Base_Weapon_Fire_3p_Simple
inputs [
 BurstRate {type=Time}
 bUseSingleVoice {type=Bool}
 Character.Affiliation {type=Enum:FortAudioAffiliation}
 Character.Health {type=Float}
 Character.UniqueID {type=Int32}
 FireBurst {type=Trigger}
 FireSingleShot {type=Trigger}
 MainFire.DistanceCurves {type=WaveTableBankAsset}
 MainFire.FalloffDistance {type=Float}
 MainFire.SoundsClose {type=WaveAsset:Array}
 MainFire.SoundsDistant {type=WaveAsset:Array}
 MainFire.SoundsMedium {type=WaveAsset:Array}
 MainFire.StopTriggerDelay {type=Time}
 Mixing.CloseGain {type=Float}
 Mixing.DistantGain {type=Float}
 Mixing.Friendly.FriendlyFireWeaponVolume {type=Float}
 Mixing.MediumGain {type=Float}
 Mixing.WeaponOutputVolume {type=Float}
 Random.EQ.A.MaxFreq {type=Float}
 Random.EQ.A.MaxGain {type=Float}
 Random.EQ.A.MinFreq {type=Float}
 Random.EQ.A.MinGain {type=Float}
 Random.MaxPitch {type=Float}
 Random.MaxVolume {type=Float}
 Random.MinPitch {type=Float}
 Random.MinVolume {type=Float}
 RetriggerDelayTime {type=Time}
 RoundsPerBurst {type=Int32}
 StartFiring {type=Trigger}
 StopBurst {type=Trigger}
 StopFiring {type=Trigger}
 UE.Attenuation.Distance {type=Float}
 UE.Source.OnPlay {type=Trigger}
 UE.Source.Orientation.Azimuth {type=Float}
 UE.Source.Orientation.Elevation {type=Float}
 UE.Spatialization.Azimuth {type=Float}
 UE.Spatialization.Elevation {type=Float}
 Weapon.AmmoRemaining {type=Int32}
 Weapon.ClipSize {type=Int32}
 Weapon.FireRate {type=Float}
 Weapon.FireStartTime {type=Float}
 Weapon.IsAimingDownSights {type=Bool}
 Weapon.IsSuppressorEnabled {type=Bool}
]
interfaces [
 Fort.Character:1.0
 Fort.Weapon:1.0
 UE.Attenuation:1.0
 UE.OutputFormat.Stereo:1.0
 UE.Source.OneShot:1.0
 UE.Source.Orientation:1.0
 UE.Source:1.1
 UE.Spatialization:1.0
]
outputs [
 bIsAlmostFinished {type=Bool}
 "UE.OutputFormat.Stereo.Audio:0" {type=Audio}
 "UE.OutputFormat.Stereo.Audio:1" {type=Audio}
 UE.Source.OneShot.OnFinished {type=Trigger}
]
pages [Default {input_defaults {
   BurstRate=0.4
   Character.Affiliation=0
   Character.Health=1
   Character.UniqueID=0
   FireBurst=null
   FireSingleShot=null
   MainFire.DistanceCurves=&WeaponRanged_DistanceCurves
   MainFire.FalloffDistance=18000
   MainFire.StopTriggerDelay=0.03
   Mixing.CloseGain=1.8
   Mixing.DistantGain=0.5
   Mixing.Friendly.FriendlyFireWeaponVolume=0.7
   Mixing.MediumGain=0.75
   Mixing.WeaponOutputVolume=0.8
   Random.EQ.A.MaxFreq=20000
   Random.EQ.A.MaxGain=0
   Random.EQ.A.MinFreq=699.9999
   Random.EQ.A.MinGain=-3
   Random.MaxPitch=1.05
   Random.MaxVolume=1.15
   Random.MinPitch=0.95
   Random.MinVolume=0.85
   RetriggerDelayTime=2
   RoundsPerBurst=2
   StartFiring=null
   StopBurst=null
   StopFiring=null
   UE.Attenuation.Distance=25000
   UE.Source.OnPlay=null
   UE.Source.Orientation.Azimuth=0
   UE.Source.Orientation.Elevation=0
   UE.Spatialization.Azimuth=0
   UE.Spatialization.Elevation=0
   Weapon.AmmoRemaining=25
   Weapon.ClipSize=25
   Weapon.FireRate=1.4286
   Weapon.FireStartTime=0
   Weapon.IsAimingDownSights=null
   Weapon.IsSuppressorEnabled=null
   bUseSingleVoice="<MetaSoundBuilderResult.SUCCEEDED: 0>"
   MainFire.SoundsClose [
    /ThermalLanceGameplay/Assets/Sounds/Waves/ThermalLance_Fire_Close_01
    /ThermalLanceGameplay/Assets/Sounds/Waves/ThermalLance_Fire_Close_02
    /ThermalLanceGameplay/Assets/Sounds/Waves/ThermalLance_Fire_Close_03
    /ThermalLanceGameplay/Assets/Sounds/Waves/ThermalLance_Fire_Close_04
   ]
   MainFire.SoundsDistant [
    /ThermalLanceGameplay/Assets/Sounds/Waves/ThermalLance_Fire_Dist_01
    /ThermalLanceGameplay/Assets/Sounds/Waves/ThermalLance_Fire_Dist_02
    /ThermalLanceGameplay/Assets/Sounds/Waves/ThermalLance_Fire_Dist_03
    /ThermalLanceGameplay/Assets/Sounds/Waves/ThermalLance_Fire_Dist_04
   ]
   MainFire.SoundsMedium [
    /ThermalLanceGameplay/Assets/Sounds/Waves/ThermalLance_Fire_Med_01
    /ThermalLanceGameplay/Assets/Sounds/Waves/ThermalLance_Fire_Med_02
    /ThermalLanceGameplay/Assets/Sounds/Waves/ThermalLance_Fire_Med_03
    /ThermalLanceGameplay/Assets/Sounds/Waves/ThermalLance_Fire_Med_04
   ]
  }}]
properties {
 assetImportData=&self:AssetImportData
 attenuationSettings=&Fort_Weapon_Projectile_NonAtt_Athena_Att
 maxDistance=27500
 numChannels=2
 outputFormat=Stereo
 qualitySetting=Fort_Weapons
 sampleRate=48000
 soundClassObject=&Fort_WeaponRanged_Fire_3P_PSM
 volume=0.6
 concurrencySet [&Weapon_Gunfire_Local_Conc, &Weapon_Gunfire_Global_Conc]
}
