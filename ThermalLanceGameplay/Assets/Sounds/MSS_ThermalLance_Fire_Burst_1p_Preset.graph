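# MetaSound Source preset for the first-person burst fire sound of the Thermal Lance weapon, built on MSS_Base_Weapon_Fire_1p_Simple.
# The parent supplies the logic, while this preset provides six close pulse waves, two-round burst timing, randomization ranges and mix levels.
# The low-ammo sound array is empty, though its threshold and volume inputs hold values.
# Output is one-shot 5.1 through the first-person ranged-fire sound class and local gunfire concurrency.
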MetaSoundSource /ThermalLanceGameplay/Assets/Sounds/MSS_ThermalLance_Fire_Burst_1p_Preset
preset_parent=/FortMetaSoundWeapon/Items/Metasounds/MSS_Base_Weapon_Fire_1p_Simple
inputs [
 bIsFiring {type=Bool}
 BurstRate {type=Time}
 Character.Affiliation {type=Enum:FortAudioAffiliation}
 Character.Health {type=Float}
 Character.UniqueID {type=Int32}
 FireBurst {type=Trigger}
 FireSingleShot {type=Trigger}
 LowAmmo.ADSMainVolumeMultiply {type=Float}
 LowAmmo.Random.MaxPitch {type=Float}
 LowAmmo.Random.MaxVolume {type=Float}
 LowAmmo.Random.MinPitch {type=Float}
 LowAmmo.Random.MinVolume {type=Float}
 LowAmmo.Sounds {type=WaveAsset:Array}
 LowAmmo.Threshold {type=Int32}
 MainFire.ADSMainVolumeMultiply {type=Float}
 MainFire.LFVolume {type=Float}
 MainFire.Random.MaxPitch {type=Float}
 MainFire.Random.MaxVolume {type=Float}
 MainFire.Random.MinPitch {type=Float}
 MainFire.Random.MinVolume {type=Float}
 MainFire.Sounds {type=WaveAsset:Array}
 MainFire.StopTriggerDelay {type=Time}
 Mixing.LowAmmoVolume {type=Float}
 Mixing.MainFireVolume {type=Float}
 Mixing.WeaponOutputVolume {type=Float}
 RetriggerDelayTime {type=Time}
 RoundsPerBurst {type=Int32}
 StartFiring {type=Trigger}
 StopBurst {type=Trigger}
 StopFiring {type=Trigger}
 UE.Source.OnPlay {type=Trigger}
 UseSingleVoice {type=Bool}
 Weapon.AmmoRemaining {type=Int32}
 Weapon.ClipSize {type=Int32}
 Weapon.FireRate {type=Float}
 Weapon.FireStartTime {type=Float}
 Weapon.IsAimingDownSights {type=Bool}
 Weapon.IsSuppressorEnabled {type=Bool}
]
interfaces [Fort.Character:1.0, Fort.Weapon:1.0, UE.OutputFormat.5dot1:1.0, UE.Source.OneShot:1.0, UE.Source:1.1]
outputs [
 bIsAlmostFinished {type=Bool}
 "UE.OutputFormat.5dot1.Audio:0" {type=Audio}
 "UE.OutputFormat.5dot1.Audio:1" {type=Audio}
 "UE.OutputFormat.5dot1.Audio:2" {type=Audio}
 "UE.OutputFormat.5dot1.Audio:3" {type=Audio}
 "UE.OutputFormat.5dot1.Audio:4" {type=Audio}
 "UE.OutputFormat.5dot1.Audio:5" {type=Audio}
 UE.Source.OneShot.OnFinished {type=Trigger}
]
pages [Default {input_defaults {
   BurstRate=0.4
   Character.Affiliation=0
   Character.Health=0
   Character.UniqueID=0
   FireBurst=null
   FireSingleShot=null
   LowAmmo.ADSMainVolumeMultiply=1
   LowAmmo.Random.MaxPitch=1.05
   LowAmmo.Random.MaxVolume=1.05
   LowAmmo.Random.MinPitch=0.95
   LowAmmo.Random.MinVolume=0.95
   LowAmmo.Threshold=3
   MainFire.ADSMainVolumeMultiply=1
   MainFire.LFVolume=1
   MainFire.Random.MaxPitch=1.05
   MainFire.Random.MaxVolume=1.05
   MainFire.Random.MinPitch=0.95
   MainFire.Random.MinVolume=0.95
   MainFire.StopTriggerDelay=0.03
   Mixing.LowAmmoVolume=1.75
   Mixing.MainFireVolume=1.25
   Mixing.WeaponOutputVolume=1.5
   RetriggerDelayTime=2
   RoundsPerBurst=2
   StartFiring=null
   StopBurst=null
   StopFiring=null
   UE.Source.OnPlay=null
   UseSingleVoice="<MetaSoundBuilderResult.SUCCEEDED: 0>"
   Weapon.AmmoRemaining=30
   Weapon.ClipSize=30
   Weapon.FireRate=1.4286
   Weapon.FireStartTime=0
   Weapon.IsAimingDownSights=null
   Weapon.IsSuppressorEnabled=null
   bIsFiring=null
   LowAmmo.Sounds []
   MainFire.Sounds [
    /ThermalLanceGameplay/Assets/Sounds/Waves/ThermalLance_Fire_Close_Pulse_01
    /ThermalLanceGameplay/Assets/Sounds/Waves/ThermalLance_Fire_Close_Pulse_02
    /ThermalLanceGameplay/Assets/Sounds/Waves/ThermalLance_Fire_Close_Pulse_03
    /ThermalLanceGameplay/Assets/Sounds/Waves/ThermalLance_Fire_Close_Pulse_04
    /ThermalLanceGameplay/Assets/Sounds/Waves/ThermalLance_Fire_Close_Pulse_05
    /ThermalLanceGameplay/Assets/Sounds/Waves/ThermalLance_Fire_Close_Pulse_06
   ]
  }}]
properties {
 assetImportData=&self:AssetImportData
 numChannels=6
 outputFormat=FiveDotOne
 qualitySetting=Fort_Weapons
 sampleRate=48000
 soundClassObject=&Fort_WeaponRanged_Fire_1P_PSM
 volume=0.6
 concurrencyOverrides {maxCount=2, resolutionRule=StopOldest}
 concurrencySet [&Weapon_Gunfire_Local_Conc]
}
