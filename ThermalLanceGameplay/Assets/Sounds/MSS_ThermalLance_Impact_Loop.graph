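# Looping mono MetaSound source for Thermal Lance beam impacts, with surface-dependent sound and ranged-weapon sound class.
# MSP_SurfaceLoopSwapper crossfades to the supplied surface loop on each surface change, with pitch, volume and fade-out inputs.
# The same trigger plays a random array-variable one-shot, though no entries or link to the Impact One Shot input appear.
# Both layers are mixed, then scaled by Base Volume adjusted for local, enemy or non-enemy affiliation.
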MetaSoundSource /ThermalLanceGameplay/Assets/Sounds/MSS_ThermalLance_Impact_Loop
inputs [
 Base Volume {type=Float}
 Character.Affiliation {type=Enum:FortAudioAffiliation}
 Character.Health {type=Float}
 Character.UniqueID {type=Int32}
 Impact One Shot {type=WaveAsset:Array}
 Surfaces.CrossfadeTime {type=Time}
 Surfaces.FadeOut {type=Trigger}
 Surfaces.OnSurfaceChanged {type=Trigger}
 Surfaces.Pitch {type=Float}
 Surfaces.Sound {type=WaveAsset}
 Surfaces.Volume {type=Float}
 UE.Source.OnPlay {type=Trigger}
]
interfaces [Fort.Character:1.0, UE.OutputFormat.Mono:1.0, UE.Source:1.1]
outputs ["UE.OutputFormat.Mono.Audio:0" {type=Audio}]
pages [Default {
  input_defaults {
   Base Volume=3
   Character.Affiliation=0
   Character.Health=0
   Character.UniqueID=0
   Surfaces.CrossfadeTime=0.3
   Surfaces.FadeOut=null
   Surfaces.OnSurfaceChanged=null
   Surfaces.Pitch=0
   Surfaces.Sound=null
   Surfaces.Volume=1
   UE.Source.OnPlay=null
   Impact One Shot []
  }
  nodes [
   ArrayRandomGet_0 = Array.Random Get:WaveAsset:Array(Next=^Surfaces.OnSurfaceChanged, In Array=VariableAccessor.WaveAsset:Array(Variable=InitVariable.WaveAsset:Array().Variable).Value)
   WavePlayer_0 = UE.Wave Player:Mono(Play=@ArrayRandomGet_0.On Next, Wave Asset=@ArrayRandomGet_0.Value)
   AudioMixerAudioMixerMono2_0 = "AudioMixer.Audio Mixer (Mono, 2)"(
    In 0=UE.Multiply:Audio by Float(
     AdditionalOperands=^Surfaces.Volume
     PrimaryOperand=/FNGameplayCues/Sounds/Fort_Audio_Blueprints/SurfaceTypeSound/MSP_SurfaceLoopSwapper(
      CrossfadeTime=^Surfaces.CrossfadeTime
      FadeOutAllSounds=^Surfaces.FadeOut
      PlayNextSound=^Surfaces.OnSurfaceChanged
      PitchShift=^Surfaces.Pitch
      NextSoundToPlay=^Surfaces.Sound
     ).Out
    ).Out
    In 1=@WavePlayer_0.Out Mono
   ) {also {defaults {Gain 0=2, Gain 1=0.7}}}
   Multiply_0 = UE.Multiply:Audio by Float(
    PrimaryOperand=@AudioMixerAudioMixerMono2_0.Out
    AdditionalOperands=&MSP_AffiliationMixFloat(In Value=^Base Volume, Affiliation=^Character.Affiliation) {also {defaults {Enemy=1.25, Local=0.8, Non-Enemy=0.5}}}.Out
   )
  ]
  links [{dst="@UE.OutputFormat.Mono.Audio:0", src=Multiply_0.Out}]
 }]
properties {
 assetImportData=&self:AssetImportData
 attenuationSettings=/ThermalLanceGameplay/Assets/Sounds/Settings/Att_ThermalLance_Beam_Impact
 bLooping=true
 bOverrideConcurrency=true
 duration=10000
 maxDistance=15000
 sampleRate=48000
 soundClassObject=&Fort_WeaponRanged
 virtualizationMode=PlayWhenSilent
 volume=0.625
 concurrencyOverrides {maxCount=3}
}
