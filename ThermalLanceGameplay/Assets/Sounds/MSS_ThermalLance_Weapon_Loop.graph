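# Stereo looping MetaSound source for the Thermal Lance weapon's close firing loop, with its own attenuation and a concurrency limit.
# A wave player starts on play, with wave, pitch and volume exposed as inputs.
# Its On Play trigger drives a fade patch, whose value is weighted louder for enemies than non-enemies by character affiliation.
# That result, multiplied by loop volume, sets the stereo mixer's gain.
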MetaSoundSource /ThermalLanceGameplay/Assets/Sounds/MSS_ThermalLance_Weapon_Loop
inputs [
 Character.Affiliation {type=Enum:FortAudioAffiliation}
 Character.Health {type=Float}
 Character.UniqueID {type=Int32}
 Loop.Pitch {type=Float}
 Loop.Sound {type=WaveAsset}
 Loop.Volume {type=Float}
 UE.Source.OnPlay {type=Trigger}
]
interfaces [Fort.Character:1.0, UE.OutputFormat.Stereo:1.0, UE.Source:1.1]
outputs [
 "UE.OutputFormat.Stereo.Audio:0" {type=Audio}
 "UE.OutputFormat.Stereo.Audio:1" {type=Audio}
]
pages [Default {
  input_defaults {
   Character.Affiliation=0
   Character.Health=0
   Character.UniqueID=0
   Loop.Pitch=0
   Loop.Sound=/ThermalLanceGameplay/Assets/Sounds/Waves/ThermalLance_Fire_Close_Loop_01
   Loop.Volume=1
   UE.Source.OnPlay=null
  }
  nodes [
   InitVariable.WaveAsset:Array()
   WavePlayer_0 = UE.Wave Player:Stereo(Pitch Shift=^Loop.Pitch, Wave Asset=^Loop.Sound, Play=^UE.Source.OnPlay) {also {defaults {Loop="<MetaSoundBuilderResult.SUCCEEDED: 0>"}}}
   n1 = &MSP_Fade(In=@WavePlayer_0.On Play) {also {defaults {Fade Time=1.25}}}
   StereoMixer_One_0 = /FortMetasound/Patches/StereoMixer_One(
    "Volume(Lin)"=UE.Multiply:Float(
     PrimaryOperand=^Loop.Volume
     AdditionalOperands=&MSP_AffiliationMixFloat(Affiliation=^Character.Affiliation, In Value=@n1.Value Out) {also {defaults {Enemy=1.25, Non-Enemy=0.5}}}.Out
    ).Out
    InL=@WavePlayer_0.Out Left
    InR=@WavePlayer_0.Out Right
   )
  ]
  links [{dst="@UE.OutputFormat.Stereo.Audio:0", src=StereoMixer_One_0.OutL}, {dst="@UE.OutputFormat.Stereo.Audio:1", src=StereoMixer_One_0.OutR}]
 }]
properties {
 assetImportData=&self:AssetImportData
 attenuationSettings=/ThermalLanceGameplay/Assets/Sounds/Settings/Att_ThermalLance_Fire
 bLooping=true
 bOverrideConcurrency=true
 duration=10000
 maxDistance=18400
 numChannels=2
 outputFormat=Stereo
 sampleRate=48000
 soundClassObject=&Fort_WeaponFoley
 volume=0.8
 concurrencyOverrides {maxCount=3}
}
