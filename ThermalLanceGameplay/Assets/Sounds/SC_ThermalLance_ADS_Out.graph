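# Sound cue for the Thermal Lance weapon's ADS-out (leaving aim-down-sights) sound.
# A Random node picks among four equally weighted wave variations, feeding a Modulator that applies a slight upward pitch variation.
# It plays at reduced volume in the gadgets sound class, uses weapon-reload attenuation, and primes on load.
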SoundCue /ThermalLanceGameplay/Assets/Sounds/SC_ThermalLance_ADS_Out
root=Modulator_0
nodes [
 {id=Modulator_0, children [Random_1], properties {pitchMax=1.1, pitchMin=1}}
 {id=Random_1, children [WavePlayer_2,WavePlayer_5,WavePlayer_6,WavePlayer_7], properties {weights [1,1,1,1]}}
 {id=WavePlayer_2, properties {soundWaveAssetPtr=/ThermalLanceGameplay/Assets/Sounds/Waves/ThermalLance_ADS_Out_01}}
 {id=WavePlayer_5, properties {soundWaveAssetPtr=/ThermalLanceGameplay/Assets/Sounds/Waves/ThermalLance_ADS_Out_02}}
 {id=WavePlayer_6, properties {soundWaveAssetPtr=/ThermalLanceGameplay/Assets/Sounds/Waves/ThermalLance_ADS_Out_03}}
 {id=WavePlayer_7, properties {soundWaveAssetPtr=/ThermalLanceGameplay/Assets/Sounds/Waves/ThermalLance_ADS_Out_04}}
]
properties {
 attenuationSettings=&Fort_Weapon_Reload_Att
 bPrimeOnLoad=true
 maxDistance=2250
 soundClassObject=&Fort_Gadgets
 volumeMultiplier=0.5
}
