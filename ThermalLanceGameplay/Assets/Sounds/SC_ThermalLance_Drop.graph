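# Sound cue for dropping the Thermal Lance, played as a one-shot effect.
# A Random node picks one of four drop wave variations with equal weight.
# Playback uses the shared weapon-drop attenuation and the gadgets sound class at reduced volume.
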SoundCue /ThermalLanceGameplay/Assets/Sounds/SC_ThermalLance_Drop
root=Random_0
nodes [
 {id=Random_0, children [WavePlayer_1,WavePlayer_2,WavePlayer_3,WavePlayer_4], properties {weights [1,1,1,1]}}
 {id=WavePlayer_1, properties {soundWaveAssetPtr=/ThermalLanceGameplay/Assets/Sounds/Waves/ThermalLance_Drop_04}}
 {id=WavePlayer_2, properties {soundWaveAssetPtr=/ThermalLanceGameplay/Assets/Sounds/Waves/ThermalLance_Drop_03}}
 {id=WavePlayer_3, properties {soundWaveAssetPtr=/ThermalLanceGameplay/Assets/Sounds/Waves/ThermalLance_Drop_02}}
 {id=WavePlayer_4, properties {soundWaveAssetPtr=/ThermalLanceGameplay/Assets/Sounds/Waves/ThermalLance_Drop_01}}
]
properties {
 attenuationSettings=&Fort_Weapon_Drop_Att
 maxDistance=3000
 soundClassObject=&Fort_Gadgets
 volumeMultiplier=0.7
}
