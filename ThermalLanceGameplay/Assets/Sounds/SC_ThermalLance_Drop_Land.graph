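# Sound cue for the landing sound of a dropped Thermal Lance.
# A Random node picks among four equally weighted landing wave variations and feeds a Modulator at the root.
# It uses the weapon-drop attenuation and the gadgets sound class, with a raised cue volume.
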SoundCue /ThermalLanceGameplay/Assets/Sounds/SC_ThermalLance_Drop_Land
root=Modulator_1
nodes [
 {id=Modulator_1, children [Random_0]}
 {id=Random_0, children [WavePlayer_0,WavePlayer_1,WavePlayer_2,WavePlayer_3], properties {weights [1,1,1,1]}}
 {id=WavePlayer_0, properties {soundWaveAssetPtr=/ThermalLanceGameplay/Assets/Sounds/Waves/ThermalLance_Drop_Land_01}}
 {id=WavePlayer_1, properties {soundWaveAssetPtr=/ThermalLanceGameplay/Assets/Sounds/Waves/ThermalLance_Drop_Land_02}}
 {id=WavePlayer_2, properties {soundWaveAssetPtr=/ThermalLanceGameplay/Assets/Sounds/Waves/ThermalLance_Drop_Land_03}}
 {id=WavePlayer_3, properties {soundWaveAssetPtr=/ThermalLanceGameplay/Assets/Sounds/Waves/ThermalLance_Drop_Land_04}}
]
properties {
 attenuationSettings=&Fort_Weapon_Drop_Att
 maxDistance=3000
 soundClassObject=&Fort_Gadgets
 volumeMultiplier=1.5
}
