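# Sound cue for equipping the Thermal Lance, played after a short fixed delay.
# A random node picks one of four equally weighted equip waves, and a modulator varies its pitch slightly upward for variation.
# It uses the weapon reload attenuation and the gadgets sound class at reduced volume, and primes on load.
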SoundCue /ThermalLanceGameplay/Assets/Sounds/SC_ThermalLance_Equip
root=Delay_0
nodes [
 {id=Delay_0, children [Modulator_0], properties {delayMax=0.1, delayMin=0.1}}
 {id=Modulator_0, children [Random_0], properties {pitchMax=1.1, pitchMin=1}}
 {id=Random_0, children [WavePlayer_0,WavePlayer_1,WavePlayer_3,WavePlayer_4], properties {weights [1,1,1,1]}}
 {id=WavePlayer_0, properties {soundWaveAssetPtr=/ThermalLanceGameplay/Assets/Sounds/Waves/ThermalLance_Equip_01}}
 {id=WavePlayer_1, properties {soundWaveAssetPtr=/ThermalLanceGameplay/Assets/Sounds/Waves/ThermalLance_Equip_02}}
 {id=WavePlayer_3, properties {soundWaveAssetPtr=/ThermalLanceGameplay/Assets/Sounds/Waves/ThermalLance_Equip_03}}
 {id=WavePlayer_4, properties {soundWaveAssetPtr=/ThermalLanceGameplay/Assets/Sounds/Waves/ThermalLance_Equip_04}}
]
properties {
 attenuationSettings=&Fort_Weapon_Reload_Att
 bPrimeOnLoad=true
 maxDistance=2250
 soundClassObject=&Fort_Gadgets
 volumeMultiplier=0.7
}
