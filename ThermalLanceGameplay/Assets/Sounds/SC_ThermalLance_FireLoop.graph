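# Looping fire sound cue for the Thermal Lance beam weapon, playing a close-perspective beam loop wave.
# A continuous modulator wraps the looping wave player and reads the Overheat parameter to raise pitch and volume as it climbs, with pitch starting to rise before volume.
# It uses the ranged weapon sound class, beam-specific attenuation, and local and global gunfire concurrency groups.
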SoundCue /ThermalLanceGameplay/Assets/Sounds/SC_ThermalLance_FireLoop
root=ModulatorContinuous_0
nodes [
 {
  id=ModulatorContinuous_0
  children [WavePlayer_0]
  properties {
   pitchModulationParams {default=0, maxOutput=2, minInput=0.6, minOutput=1, parameterName=Overheat}
   volumeModulationParams {default=0, maxOutput=1.5, minInput=0.76, minOutput=1, parameterName=Overheat}
  }
 }
 {
  id=WavePlayer_0
  properties {bLooping=true, soundWaveAssetPtr=/ThermalLanceGameplay/Assets/Sounds/Waves/ThermalLance_Beam_Beam_Close_Loop_01}
 }
]
properties {
 attenuationSettings=/ThermalLanceGameplay/Assets/Sounds/Settings/Att_ThermalLance_Beam
 bDebug=true
 maxDistance=15000
 soundClassObject=&Fort_WeaponRanged
 volumeMultiplier=1.3
 concurrencySet [&Weapon_Gunfire_Local_Conc, &Weapon_Gunfire_Global_Conc]
}
