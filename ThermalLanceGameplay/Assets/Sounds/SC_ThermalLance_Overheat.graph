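# Sound cue for the Thermal Lance weapon's overheat, layering a main overheat sound with a delayed ending sound.
# A Mixer combines the main wave, passed through a default Modulator, with the louder end wave held back by a fixed Delay of under three seconds.
# It uses the weapons sound class, references the lance's fire attenuation asset and overrides concurrency to block new instances at the limit.
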SoundCue /ThermalLanceGameplay/Assets/Sounds/SC_ThermalLance_Overheat
root=Mixer_0
nodes [
 {id=Delay_0, children [WavePlayer_0], properties {delayMax=2.7, delayMin=2.7}}
 {id=Mixer_0, children [Modulator_0,Delay_0], properties {inputVolume [1,1.5]}}
 {id=Modulator_0, children [WavePlayer_1]}
 {id=WavePlayer_0, properties {soundWaveAssetPtr=/ThermalLanceGameplay/Assets/Sounds/Waves/ThermalLance_Overheat_End_01}}
 {id=WavePlayer_1, properties {soundWaveAssetPtr=/ThermalLanceGameplay/Assets/Sounds/Waves/ThermalLance_Overheat_01}}
]
properties {
 attenuationSettings=/ThermalLanceGameplay/Assets/Sounds/Settings/Att_ThermalLance_Fire
 bOverrideConcurrency=true
 maxDistance=18400
 soundClassObject=/Game/Sounds/FortSoundClassesAndModes/Fort_Weapons
 volumeMultiplier=0.7
 attenuationOverrides {distanceAlgorithm=NaturalSound, falloffDistance=4000, attenuationShapeExtents {x=1000}}
 concurrencyOverrides {resolutionRule=PreventNew}
}
