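# Sound cue for picking up the Thermal Lance.
# A random node picks one of four pickup wave variations with equal weight.
# A modulator on top applies a slight random pitch shift to vary repeated plays.
# The cue uses the Fort weapon pickup attenuation and plays under the gadgets sound class.
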SoundCue /ThermalLanceGameplay/Assets/Sounds/SC_ThermalLance_Pickup
root=Modulator_0
nodes [
 {id=Modulator_0, children [Random_0], properties {pitchMax=1.02, pitchMin=0.98}}
 {id=Random_0, children [WavePlayer_1,WavePlayer_2,WavePlayer_3,WavePlayer_4], properties {weights [1,1,1,1]}}
 {id=WavePlayer_1, properties {soundWaveAssetPtr=/ThermalLanceGameplay/Assets/Sounds/Waves/ThermalLance_Pickup_04}}
 {id=WavePlayer_2, properties {soundWaveAssetPtr=/ThermalLanceGameplay/Assets/Sounds/Waves/ThermalLance_Pickup_03}}
 {id=WavePlayer_3, properties {soundWaveAssetPtr=/ThermalLanceGameplay/Assets/Sounds/Waves/ThermalLance_Pickup_02}}
 {id=WavePlayer_4, properties {soundWaveAssetPtr=/ThermalLanceGameplay/Assets/Sounds/Waves/ThermalLance_Pickup_01}}
]
properties {
 attenuationSettings=&Fort_Weapon_Pickup_Att
 maxDistance=3500
 soundClassObject=&Fort_Gadgets
 volumeMultiplier=1
}
