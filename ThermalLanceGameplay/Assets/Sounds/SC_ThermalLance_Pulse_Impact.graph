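# Sound cue for the Thermal Lance pulse impact explosion, layered by distance.
# A Mixer blends close, medium and distant layers, each using its own Fortnite explosion attenuation asset, with medium and distant boosted relative to close.
# Each layer runs through a Modulator and randomly picks one of several equally weighted wave variations.
# The cue is reduced in volume, routed to a third-person ranged weapon fire sound class, and overrides concurrency to cap simultaneous instances.
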SoundCue /ThermalLanceGameplay/Assets/Sounds/SC_ThermalLance_Pulse_Impact
root=Mixer_5
nodes [
 {id=Attenuation_0, children [Modulator_0], properties {attenuationSettings=/Game/Sounds/Attenuation/Fort_Explosion_Close_Athena_Att}}
 {id=Attenuation_1, children [Modulator_1], properties {attenuationSettings=/Game/Sounds/Attenuation/Fort_Explosion_Medium_Athena_Att}}
 {id=Attenuation_2, children [Modulator_3], properties {attenuationSettings=/Game/Sounds/Attenuation/Fort_Explosion_Distant_Athena_Att}}
 {id=Mixer_5, children [Attenuation_0, Attenuation_1, Attenuation_2], properties {inputVolume [1,2,2]}}
 {id=Modulator_0, children [Random_5]}
 {id=Modulator_1, children [Random_7]}
 {id=Modulator_3, children [Random_8]}
 {
  id=Random_5
  children [WavePlayer_18, WavePlayer_19, WavePlayer_20, WavePlayer_21, WavePlayer_25, WavePlayer_26]
  properties {weights [1,1,1,1,1,1]}
 }
 {id=Random_7, children [WavePlayer_27, WavePlayer_28, WavePlayer_29, WavePlayer_30], properties {weights [1,1,1,1]}}
 {id=Random_8, children [WavePlayer_31, WavePlayer_32, WavePlayer_33, WavePlayer_34], properties {weights [1,1,1,1]}}
 {id=WavePlayer_18, properties {soundWaveAssetPtr=/ThermalLanceGameplay/Assets/Sounds/Waves/ThermalLance_Pulse_Expl_Close_01}}
 {id=WavePlayer_19, properties {soundWaveAssetPtr=/ThermalLanceGameplay/Assets/Sounds/Waves/ThermalLance_Pulse_Expl_Close_02}}
 {id=WavePlayer_20, properties {soundWaveAssetPtr=/ThermalLanceGameplay/Assets/Sounds/Waves/ThermalLance_Pulse_Expl_Close_03}}
 {id=WavePlayer_21, properties {soundWaveAssetPtr=/ThermalLanceGameplay/Assets/Sounds/Waves/ThermalLance_Pulse_Expl_Close_04}}
 {id=WavePlayer_25, properties {soundWaveAssetPtr=/ThermalLanceGameplay/Assets/Sounds/Waves/ThermalLance_Pulse_Expl_Close_05}}
 {id=WavePlayer_26, properties {soundWaveAssetPtr=/ThermalLanceGameplay/Assets/Sounds/Waves/ThermalLance_Pulse_Expl_Close_06}}
 {id=WavePlayer_27, properties {soundWaveAssetPtr=/ThermalLanceGameplay/Assets/Sounds/Waves/ThermalLance_Pulse_Expl_Med_01}}
 {id=WavePlayer_28, properties {soundWaveAssetPtr=/ThermalLanceGameplay/Assets/Sounds/Waves/ThermalLance_Pulse_Expl_Med_02}}
 {id=WavePlayer_29, properties {soundWaveAssetPtr=/ThermalLanceGameplay/Assets/Sounds/Waves/ThermalLance_Pulse_Expl_Med_03}}
 {id=WavePlayer_30, properties {soundWaveAssetPtr=/ThermalLanceGameplay/Assets/Sounds/Waves/ThermalLance_Pulse_Expl_Med_04}}
 {id=WavePlayer_31, properties {soundWaveAssetPtr=/ThermalLanceGameplay/Assets/Sounds/Waves/ThermalLance_Pulse_Expl_Dist_01}}
 {id=WavePlayer_32, properties {soundWaveAssetPtr=/ThermalLanceGameplay/Assets/Sounds/Waves/ThermalLance_Pulse_Expl_Dist_02}}
 {id=WavePlayer_33, properties {soundWaveAssetPtr=/ThermalLanceGameplay/Assets/Sounds/Waves/ThermalLance_Pulse_Expl_Dist_03}}
 {id=WavePlayer_34, properties {soundWaveAssetPtr=/ThermalLanceGameplay/Assets/Sounds/Waves/ThermalLance_Pulse_Expl_Dist_04}}
]
properties {
 bOverrideConcurrency=true
 maxDistance=27500
 soundClassObject=&Fort_WeaponRanged_Fire_3P_PSM
 volumeMultiplier=0.7
 concurrencyOverrides {maxCount=3}
}
