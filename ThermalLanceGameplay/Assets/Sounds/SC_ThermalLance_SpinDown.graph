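# Sound cue for the Thermal Lance weapon's spin-down, built on its fire-end sound wave.
# A single wave player feeds a modulator at the root, which is left at its usual pitch and volume settings.
# It uses the Thermal Lance fire attenuation and three concurrency groups: local gunfire, global gunfire and a dedicated spin-down one.
# The cue's debug flag is switched on.
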SoundCue /ThermalLanceGameplay/Assets/Sounds/SC_ThermalLance_SpinDown
root=Modulator_3
nodes [
 {id=Modulator_3, children [WavePlayer_3]}
 {id=WavePlayer_3, properties {soundWaveAssetPtr=/ThermalLanceGameplay/Assets/Sounds/Waves/ThermalLance_Fire_End_01}}
]
properties {
 attenuationSettings=/ThermalLanceGameplay/Assets/Sounds/Settings/Att_ThermalLance_Fire
 bDebug=true
 maxDistance=18400
 volumeMultiplier=1
 concurrencySet [
  &Weapon_Gunfire_Local_Conc
  &Weapon_Gunfire_Global_Conc
  &/SCC_ThermalLance_Spindown
 ]
}
