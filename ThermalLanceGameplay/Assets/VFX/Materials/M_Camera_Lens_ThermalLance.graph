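# Unlit translucent Niagara mesh-particle material, apparently a Thermal Lance camera lens effect, drawn without depth testing.
# Panning inky smoke and Color02-tinted plasma coils are each sampled at three offset UVs into RGB, creating chromatic fringing.
# Their sum is tinted by particle color and fed through MF_FX_EmissiveAdjustments.
# Particle alpha multiplies a vertical UV falloff mask; UV_Detail and UV_Base scale tiling uniformly using only R.
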Material /ThermalLanceGameplay/Assets/VFX/Materials/M_Camera_Lens_ThermalLance
graph {
 properties {
  bDisableDepthTest=true
  bUsedWithNiagaraMeshParticles=true
  blendMode=BLEND_Translucent
  pixelDepthOffsetMode=PDOM_Legacy
  shadingModel=MSM_Unlit
  usedShadingModels=Unlit
 }
 nodes [
  Constant3Vector() {constant=<rgba 1.6177 13 12.612 1>}
  n1 = ParticleColor()
  n2 = VectorParameter() {parameterName=UV_Detail, defaultValue=<rgba 2 1 0 0>}
  n3 = Multiply(
   A=TextureCoordinate()
   B=MaterialFunctionCall(X=@n2.R, Y=@n2.R) {materialFunction=&MakeFloat2}.Result
  )
  n4 = Constant() {r=0.005}
  n5 = Add(A=@n3, B=@n4)
  n6 = Subtract(A=@n3, B=@n4)
  n7 = MaterialFunctionCall(
   X=TextureSample(UVs=Panner(Coordinate=@n5) {speedY=-0.5}) {texture=&T_Inky_Smoke_Tile}.R
   Y=TextureSample(UVs=Panner(Coordinate=@n3) {speedY=-0.5}) {texture=&T_Inky_Smoke_Tile}.R
   Z=TextureSample(UVs=Panner(Coordinate=@n6) {speedY=-0.5}) {texture=&T_Inky_Smoke_Tile}.R
  ) {materialFunction=&MakeFloat3}
  n8 = VectorParameter() {parameterName=UV_Base, defaultValue=<rgba 3 0.5 0 0>}
  n9 = Multiply(
   A=TextureCoordinate()
   B=MaterialFunctionCall(X=@n8.R, Y=@n8.R) {materialFunction=&MakeFloat2}.Result
  )
  n10 = Constant() {r=0.01}
  n11 = Constant2Vector() {g=-1}
  n12 = TextureSample(UVs=Panner(Coordinate=Add(A=@n9, B=@n10), Speed=@n11) {speedY=-0.5}) {samplerType=SAMPLERTYPE_LinearGrayscale, texture=/Game/Effects/Fort_Effects/Emotes/UltraLight/Texture/T_PlasmaCoils}
  n13 = TextureSample(UVs=Panner(Coordinate=@n9, Speed=@n11) {speedY=-0.5}) {samplerType=SAMPLERTYPE_LinearGrayscale, texture=&T_PlasmaCoils}
  n14 = TextureSample(UVs=Panner(Coordinate=Subtract(A=@n9, B=@n10), Speed=@n11) {speedY=-0.5}) {samplerType=SAMPLERTYPE_LinearGrayscale, texture=&T_PlasmaCoils}
  n15 = Multiply(
   A=VectorParameter() {parameterName=Color02, defaultValue=<rgba 0.2604 0.5069 1 1>}.RGB
   B=Power(Base=MaterialFunctionCall(X=@n12.R, Y=@n13.R, Z=@n14.R) {materialFunction=&MakeFloat3}.Result) {constExponent=1.5}
  )
  n16 = Add(A=Multiply(A=Power(Base=@n7.Result) {constExponent=3}, B=Constant3Vector() {constant=<rgba 1 1 1 0>}), B=@n15)
  MaterialFunctionCall(Emissive=Multiply(A=@n1.RGB, B=Multiply(A=@n16))) {materialFunction=&MF_FX_EmissiveAdjustments}
  n17 = TextureCoordinate()
  n18 = Power(Base=OneMinus(None=ComponentMask(None=@n17) {g=true}))
  n19 = Power(Base=ComponentMask(None=TextureCoordinate()) {g=true}) {constExponent=12}
  n20 = Multiply(A=Saturate(None=Multiply(A=@n18) {constB=5}), B=Saturate(None=Multiply(A=@n19) {constB=50}))
  Multiply(A=Multiply(A=@n1.A, B=Multiply(A=@n20) {constB=20}))
  Multiply(A=@n15) {constB=5}
  VectorParameter() {parameterName=Color01, defaultValue=<rgba 0.0036 0.0555 0.1 1>}
 ]
}
