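# Translucent unlit deferred-decal material for Niagara ribbons, named as a molten-ground effect for Thermal Lance.
# A noise-perturbed V-coordinate SphereMask, multiplied by particle alpha, feeds a DepthFade for soft-edged opacity.
# The graph also holds power-curve color gradients along UV1, BlackBody heat colors, particle-tinted noise and an Apollo landscape virtual texture sample.
# No connections from these branches to material outputs appear.
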Material /ThermalLanceGameplay/Assets/VFX/Materials/M_MoltenGround_Ribbon
graph {
 properties {
  bAutomaticallySetUsageInEditor=false
  bUsedWithNiagaraRibbons=true
  blendMode=BLEND_Translucent
  materialDomain=MD_DeferredDecal
  shadingModel=MSM_Unlit
  twoSided=true
  usedShadingModels=Unlit
 }
 nodes [
  n1 = OneMinus(None=ComponentMask(None=TextureCoordinate() {coordinateIndex=1}) {r=true})
  n2 = ScalarParameter() {defaultValue=16, parameterName=Param_33}
  n3 = OneMinus(None=ComponentMask(None=TextureCoordinate() {coordinateIndex=1}) {r=true})
  n4 = ScalarParameter() {defaultValue=4, parameterName=Param_3}
  Add(
   A=LinearInterpolate(B=VectorParameter() {parameterName=Param_2, defaultValue=<rgba 7.5 0.125 0 1>}.RGB, Alpha=Power(Base=@n1, Exp=@n2))
   B=LinearInterpolate(B=VectorParameter() {parameterName=Param_1}.RGB, Alpha=Power(Base=@n3, Exp=@n4))
  )
  n5 = ComponentMask(None=TextureCoordinate()) {g=true}
  n6 = OneMinus(None=@n5)
  BlackBody!(Temp=LinearInterpolate(B=ScalarParameter() {defaultValue=17500, parameterName=Param_1}, Alpha=Multiply(A=@n5, B=@n6)))
  BlackBody(Temp=ScalarParameter() {defaultValue=2500, parameterName=Param_4})
  Constant()
  Constant()
  n7 = ParticleColor()
  n8 = Multiply(
   A=TextureCoordinate() {coordinateIndex=1}
   B=AppendVector(A=ScalarParameter() {defaultValue=5, parameterName=X}, B=ScalarParameter() {defaultValue=1, parameterName=Y})
  )
  n9 = Add(
   A=Multiply(
    A=TextureSample(UVs=@n8) {texture=/ThermalLanceGameplay/Athena/Effects/Particles/Skydiving/NS_Hightower_Date/T_AtmosphericCloudNoise_Vertical}.G
    B=ScalarParameter() {defaultValue=0.15}
   )
   B=ComponentMask(None=TextureCoordinate()) {g=true}
  )
  n10 = ScalarParameter() {defaultValue=0.5, parameterName=Mask Loc}
  n11 = ScalarParameter() {defaultValue=0.396, parameterName=Mask Rad}
  n12 = ScalarParameter() {defaultValue=0.6248, parameterName=Mask Hard}
  DepthFade(
   Opacity=Multiply(A=@n7.A, B=SphereMask(A=@n9, B=@n10, Radius=@n11, Hardness=@n12))
   FadeDistance=ScalarParameter() {defaultValue=5, parameterName=Param_5}
  )
  DynamicParameter() {paramNames [Radius,Param2,Param3,Param4]}
  LinearInterpolate()
  n13 = Multiply(
   A=TextureCoordinate()
   B=AppendVector(A=ScalarParameter() {defaultValue=10, parameterName=X2}, B=ScalarParameter() {defaultValue=1, parameterName=Y2})
  )
  Multiply(
   A=@n7.RGB
   B=LinearInterpolate(B=TextureSample(UVs=@n13) {texture=&T_AtmosphericCloudNoise_Vertical}.RGB)
  )
  Multiply(
   A=VectorParameter().RGB
   B=RuntimeVirtualTextureSample!(UVs, World Position=WorldPosition().XYZ) {materialType=BaseColor_Normal_Specular_YCoCg, virtualTexture=/Game/Athena/Apollo/Maps/Landscape/VT/RVT_Apollo_Landscape}.BaseColor
  )
  OneMinus()
  ParticlePositionWS()
  TextureCoordinate()
 ]
}
