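# Additive, unlit, two-sided Niagara material for the Thermal Lance overheat glow, enabled for sprites, ribbons and mesh particles.
# An energy streak texture, tiled by the Tile parameter and panned along V, feeds a lerp weighted by inverted V, fading along the UV.
# Separate branches scale the vertex normal by V and a scalar, and read Particle Color.
# The graph shows no links from these branches to material outputs.
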Material /ThermalLanceGameplay/Assets/VFX/Materials/M_ThermalLance_OverheatGlow
graph {
 properties {
  bAutomaticallySetUsageInEditor=false
  bUsedWithNiagaraMeshParticles=true
  bUsedWithNiagaraRibbons=true
  bUsedWithNiagaraSprites=true
  blendMode=BLEND_Additive
  shadingModel=MSM_Unlit
  twoSided=true
  usedShadingModels=Unlit
 }
 nodes [
  n1 = Multiply(A=TextureCoordinate(), B=ComponentMask(None=VectorParameter() {parameterName=Tile, defaultValue=<rgba 5 0.1 0 0>}.RGB) {g=true, r=true})
  n2 = ComponentMask(None=TextureCoordinate()) {g=true}
  LinearInterpolate(
   B=TextureSample(UVs=Panner(Coordinate=@n1) {speedY=-1}) {texture=/Game/Effects/Fort_Effects/Textures/Energy/T_OutpostShockwaveStreaks_02}.G
   Alpha=OneMinus(None=@n2)
  )
  Multiply(A=Multiply(A=@n2, B=VertexNormalWS()), B=ScalarParameter() {defaultValue=0.25})
  ParticleColor()
 ]
}
