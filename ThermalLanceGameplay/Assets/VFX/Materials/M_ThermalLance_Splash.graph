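# Masked, Default Lit material for Niagara sprites in the Thermal Lance splash effect, built on shared Fort water splash textures.
# A lighting term dots the splash normal map with the atmospheric light vector, transformed from world to tangent space, then saturates it.
# An erosion term subtracts inverted particle alpha from the splash texture's green channel, so the shape shrinks as particle alpha falls.
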Material /ThermalLanceGameplay/Assets/VFX/Materials/M_ThermalLance_Splash
graph {
 properties {bAutomaticallySetUsageInEditor=false, bUsedWithNiagaraSprites=true, blendMode=BLEND_Masked, usedShadingModels=Default Lit}
 nodes [
  n1 = AtmosphericLightVector()
  Saturate(None=DotProduct(
    A=TextureSample() {samplerType=SAMPLERTYPE_Normal, texture=/Game/Effects/Fort_Effects/Textures/Water/T_waterSplash_R_N}.RGB
    B=Transform(None=@n1) {transformSourceType=TRANSFORMSOURCE_World, transformType=TRANSFORM_Tangent}
   ))
  Subtract(
   A=Saturate(None=TextureSample() {samplerType=SAMPLERTYPE_LinearColor, texture=/Game/Effects/Fort_Effects/Textures/Water/T_waterSplash}.G)
   B=OneMinus(None=ParticleColor().A)
  )
 ]
}
