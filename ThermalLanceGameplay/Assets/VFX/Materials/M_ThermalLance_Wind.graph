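# Translucent, unlit, two-sided wind material for Thermal Lance VFX on Niagara mesh particles.
# Two counter-panned noise and normal-map samples warp the UVs of a cloud texture, which particle color tints.
# Opacity branches shape the cloud with UV edge masks, a dynamic-parameter-panned gradient and particle alpha, then fade by distance, view angle and depth.
# Other branches build emissive with scene color (non-mobile), Fresnel and time-of-day boost.
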Material /ThermalLanceGameplay/Assets/VFX/Materials/M_ThermalLance_Wind
graph {
 properties {
  bTangentSpaceNormal=false
  bUsedWithNiagaraMeshParticles=true
  blendMode=BLEND_Translucent
  opacityMaskClipValue=0.999
  pixelDepthOffsetMode=PDOM_Legacy
  shadingModel=MSM_Unlit
  translucentSelfShadowDensityScale=0
  translucentShadowDensityScale=0
  twoSided=true
  usedShadingModels=Unlit
 }
 nodes [
  n1 = ParticleColor()
  n2 = Panner(Coordinate=TextureCoordinate() {uTiling=0.5, vTiling=0.5}) {speedX=-0.5, speedY=-0.5}
  n3 = Panner(Coordinate=TextureCoordinate() {uTiling=0.5, vTiling=0.5}) {speedX=0.5, speedY=-0.65}
  n4 = Multiply(A=Multiply(
    A=TextureSample(UVs=@n2) {texture=&2ndInfectedMapLinear}.R
    B=TextureSample(UVs=@n3) {texture=&2ndInfectedMapLinear}.R
   )) {constB=2}
  n5 = Clamp(None=OneMinus(None=@n4))
  n6 = TextureSample(UVs=@n2) {samplerType=SAMPLERTYPE_Normal, texture=/Game/Global/Textures/Default/Tileable_Patterns/2ndInfectedMapNormal}
  n7 = TextureSample(UVs=@n3) {samplerType=SAMPLERTYPE_Normal, texture=&2ndInfectedMapNormal}
  n8 = ComponentMask(None=LinearInterpolate(A=@n5, B=@n4, Alpha=Multiply(A=@n6.R, B=@n7.R))) {g=true, r=true}
  n9 = Panner(Coordinate=Add(A=Multiply(A=@n8) {constB=0.12}, B=TextureCoordinate()), Time=Constant() {r=-2}) {speedX=0.05, speedY=0.05}
  n10 = TextureSample(UVs=@n9) {samplerType=SAMPLERTYPE_LinearColor, texture=/Game/Packages/Fortress_Transylvania/Haunted_Forest/Swamps/Textures/CloudRenderHeightMap}
  n11 = Multiply(A=@n10.R, B=@n1.RGB)
  Add(A=Multiply(A=@n1.RGB) {constB=0.5}, B=Multiply(A=@n11))
  n12 = StaticBool() {value=true}
  n13 = MaterialFunctionCall(Invert Fresnel=@n12, Use Cheap contrast=@n12) {materialFunction=&Fresnel_Function}
  Clamp(None=Multiply(A=Power(Base=@n13.Result) {constExponent=5}) {constB=7})
  n14 = ComponentMask(None=TextureCoordinate()) {r=true}
  n15 = ComponentMask(None=TextureCoordinate()) {g=true}
  n16 = Multiply(A=Subtract(A=@n14), B=Subtract(A=@n15))
  n17 = Multiply(A=Subtract(A=@n14) {constB=0}, B=Subtract(A=@n15) {constB=0})
  n18 = Multiply(A=Power(Base=Multiply(A=@n16, B=@n17))) {constB=150}
  n19 = DynamicParameter() {desc=0 to -.5 for contrast control, defaultValue=<rgba 0 0 0 1>, paramNames [PanMask,LerpDetails,Param3,Param4]}
  n20 = MaterialFunctionCall(UV Channel=Add(A=@n9, B=@n19.PanMask) {constB=0, desc=0 to -1 to pan Mask}) {materialFunction=&LinearGradient}
  n21 = Multiply(A=OneMinus(None=@n20.VGradient), B=@n20.VGradient)
  n22 = Multiply(A=Multiply(A=Clamp(None=@n18), B=LinearInterpolate(Alpha=@n21) {constB=4}), B=@n10.R)
  n23 = LinearInterpolate(
   B=Multiply(A=Multiply(A=@n22, B=@n5), B=@n1.A)
   Alpha=MaterialFunctionCall!(
    Blend Range=ScalarParameter() {defaultValue=200, parameterName=DistanceBlend}
    Start Offset=ScalarParameter() {defaultValue=-100, parameterName=DistanceOffset}
   ) {materialFunction=/Engine/Functions/Engine_MaterialFunctions03/Particles/Distance_Blend}.Result
  )
  n24 = Constant() {r=1}
  DepthFade(Opacity=Multiply(
    A=@n23
    B=MaterialFunctionCall("FadeSensitivity (optional)"=@n24) {materialFunction=&FadeBasedOnViewAngle}.Result
   )) {fadeDistanceDefault=50}
  n25 = CollectionParameter() {
   collection=&FortniteMaterialParameters
   parameterName=FortniteActiveTimeOfDay
  }
  LinearInterpolate(A=@n23, B=Multiply(A=@n23) {constB=2}, Alpha=DotProduct(A=@n25, B=Constant4Vector() {constant=<rgba 1 1 1 0>}.RGBA)) {desc="A=Night B=Day"}
  n26 = SceneColor!(UVs)
  n27 = Constant()
  MaterialFunctionCall(Emissive=Add(A=ShadingPathSwitch(Default=@n26.RGB, Mobile=@n27) {desc=Migrated from FeatureLevelSwitch}, B=@n11) {constB=0.1}) {materialFunction=&MF_FX_EmissiveAdjustments}
  TextureCoordinate()
 ]
}
